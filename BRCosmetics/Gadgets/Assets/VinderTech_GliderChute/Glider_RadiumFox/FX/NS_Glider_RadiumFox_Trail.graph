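# Niagara system for the RadiumFox glider trail: flame, glow, smoke and ember effects at six thruster sockets, plus wing-tip style ribbon trails.
# Emitters sample the glider skeletal mesh: mesh flame cores and jets, flickering sprite glows with lights, mesh smoke, GPU embers and AnimTrail ribbons.
# User.ForwardDot modulates spawning and flame intensity, system-age curves add a startup flare, and time-of-day and front-end modifiers adjust brightness.
# The Empty and SmokeLarge_Local emitters are disabled.
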NiagaraSystem /BRCosmetics/Gadgets/Assets/VinderTech_GliderChute/Glider_RadiumFox/FX/NS_Glider_RadiumFox_Trail
dependencies {
 usedDataInterfaces [
  &NiagaraDataInterfaceCurve
  &NiagaraDataInterfaceColorCurve
  &NiagaraDataInterfaceArrayFloat3
  &NiagaraDataInterfaceVectorCurve
  &NiagaraDataInterfaceArrayPosition
  &NiagaraDataInterfaceVector2DCurve
  &NiagaraDataInterfaceArrayInt32
 ]
 usedDynamicInputs [&Boolean_AndOperation, &SetBoolByFloatComparison]
 usedModules [
  &self:SystemSpawnScript.SystemScriptSource.SystemScriptGraph.NiagaraNodeAssignment_0.SetVariables_#1
  &SystemState
  &Timeline
  &FN_Integrator
 ]
}
emitters {
 EmbersGPUHigh {
  simTarget=GPUComputeSim
  emitterUpdateScript [
   &EmitterState()
   &SpawnBurst_Instantaneous(
    Spawn Count=&Multiply_Int(A=&RandomRangeInt(Minimum=6, Maximum=10), B=6)
    Use Spawn Probability=false
    UseLoopCountLimit=true
   )
   &SpawnRate(
    SpawnGroup=1
    SpawnRate=&Lerp_Float(A=0, B=&Multiply_Float(A=5, B=6), Alpha=&ClampFloat(Float=$User.ForwardDot))
    Use Spawn Probability=true
   )
  ]
  particleSpawnScript [
   &FN_SetSpawnGroup(Number of Spawn Groups=3)
   /SetVariables(Particles.PerParticleRandom:NiagaraFloat=&RandomRangeFloat(Minimum=0, Maximum=1)) {isSetParameters=true}
   &InitializeParticle(
    Color Mode=::NewEnumerator0
    Lifetime=&Lerp_Float(A=0.35, B=0.5, Alpha=$Particles.PerParticleRandom)
    Position Offset=&SelectVectorFromArray(
     Direct Array Index=&SpawnGroup()
     Vector Selection Array=DataInterface() {propertyValues {FloatData [<0 180 -19>, <59 140 -47>, <-59 140 -47>]}}
    )
    Sprite Size=&Multiply_Vector2D(
     A=<0.75 1>
     B=&MakeVector2D(
      X=&FloatFromCurve(
       FloatCurve=DataInterface() {propertyValues {Curve {keys [<at 0 1>,<at 0.8 2 RCIM_Constant>,<at 0.95 3 RCIM_Constant>]}}}
       CurveIndex=$Particles.PerParticleRandom
      )
      Y=&FloatFromCurve(
       FloatCurve=DataInterface() {propertyValues {Curve {keys [<at 0 1>,<at 0.8 2 RCIM_Constant>,<at 0.95 3>]}}}
       CurveIndex=$Particles.PerParticleRandom
      )
     )
    )
    Sprite Size Mode=::NewEnumerator4
    UsePositionOffset=false
    Write Material Random=true
   )
   &SkeletalMeshLocation(
    Bone / Socket Index=&Add_Integer(A=&ModuloInt(A=&ReturnParticleIDAttributes(), B=6), B=4)
    Interpolate Mesh Sampling=false
    MaskSpawnGroup=false
    Mesh Sampling Type=::NewEnumerator5
    Orientation Sampling=::NewEnumerator1
    Sampled Position Offset Coordinate Space:ENiagaraSkelMeshTransforms=::NewEnumerator2
    Skeletal Mesh=$System.Skeletal Mesh
    Socket Sampling Mode=::NewEnumerator5
    Use Sampled Mesh Attribute as Ribbon ID=false
   )
   &ShapeLocation(
    Cylinder Height=0.01
    Cylinder Radius=5
    MaskSpawnGroup=false
    Override Seed=false
    Rotation Axis=&ReturnMeshOrientationAxis(Mesh Orientation Axis=::XAxis, Mesh Orientation Quat=$Particles.SkeletalMeshLocation.SampledRotation)
    Rotation Mode=::NewEnumerator1
    Shape Primitive=::NewEnumerator1
    Surface Only=false
    UseOwnerScale=true
   )
   &AddVelocity(
    Cone Angle=10
    Cone Axis=<0 1 0>
    Module.UseConeFalloff=false
    Module.UseVelDistribution=false
    Override Seed=false
    Velocity Mode=::NewEnumerator2
    Velocity Speed=&Lerp_Float(A=0, B=1200, Alpha=$Particles.PerParticleRandom)
   )
   &InheritVelocity(
    Inherited Velocity Amount Scale=&VectorFromFloat(Value=0)
    Inherited Velocity Speed Limit=555
    Speed Limit=true
    UseSpeedThreshold=false
   )
   &CameraOffset()
  ]
  particleUpdateScript [
   &ParticleState()
   &AccelerationForce(Acceleration=<0 0 50>)
   &Drag(
    Drag=&Lerp_Float(
     A=5
     B=10
     Alpha=&FloatFromCurve(
      FloatCurve=DataInterface() {propertyValues {Curve {keys [<at 0 0 RCIM_Constant>, {tangentMode=RCTM_User, time=0.05, value=1}]}}}
      CurveIndex=$Engine.System.Age
     )
    )
    Use Linear Drag=true
    Use Rotational Drag=false
   )
   &CurlNoiseForce(Noise Frequency=50000, Noise Quality / Cost=::NewEnumerator0, Noise Strength=1500, Pan Noise=true, Pan Noise Field=<100 100 100>)
   &SolveForcesAndVelocity(Clamp Velocity=false, Limit Acceleration=false)
   &ScaleSpriteSizeBySpeed(
    Max Scale Factor=<0.9 9>
    Sample Scale Factor By Curve=true
    Scale Factor Curve=DataInterface() {propertyValues {Curve {keys [<at 0 0>, {arriveTangent=1.4989, interpMode=RCIM_Cubic, tangentMode=RCTM_Break, time=1, value=1}]}}}
    Velocity Threshold=1200
   )
   &ScaleSpriteSize(
    Uniform Curve Scale=&MakeFloatFromVector(Vector=$Engine.Owner.Scale, Channel=::NewEnumerator0)
    Uniform Curve Sprite Scale=DataInterface() {propertyValues {Curve {keys [
        <at 0.15 1>
        {arriveTangent=-0.0405, interpMode=RCIM_Cubic, leaveTangent=-0.0405, tangentMode=RCTM_User, time=0.7009, value=1}
        {arriveTangent=-8.0089, interpMode=RCIM_Cubic, leaveTangent=-8.0089, tangentMode=RCTM_User, time=1, value=0}
       ]}}}
   )
   /SetVariables(Particles.Color=&MultiplyLinearColor(
     Linear Color=&ColorFromCurve(ColorCurve=DataInterface() {propertyValues {
        AlphaCurve {keys [<at 0 1 RCIM_Cubic>]}
        BlueCurve {keys [<at 0 0 RCIM_Cubic>,<at 0.7551 0 RCIM_Cubic>]}
        GreenCurve {keys [{arriveTangent=-6.9718, interpMode=RCIM_Cubic, leaveTangent=-6.9718, tangentMode=RCTM_User, time=0, value=3.6667}, <at 0.7551 0 RCIM_Cubic>]}
        RedCurve {keys [{arriveTangent=-68.5078, interpMode=RCIM_Cubic, leaveTangent=-68.5078, tangentMode=RCTM_User, time=0, value=20}, <at 0.7551 1 RCIM_Cubic>]}
       }})
     Scale Factor=&MakeVector4_FromFloat(Float=&FNTimeOfDayScalar(Night Scale=0.3, Morning Scale=0.9, Evening Scale=0.9))
    )) {isSetParameters=true}
  ]
  rendererClasses [&NiagaraSpriteRendererProperties]
  renderers [{rendererClass=&NiagaraSpriteRendererProperties}]
 }
 Empty {
  bEnabled=false
  emitterUpdateScript [
   &EmitterState()
   &SpawnBurst_Instantaneous(Use Spawn Probability=false, UseLoopCountLimit=false)
  ]
  particleSpawnScript [&InitializeParticle(Color Mode=::NewEnumerator0, Lifetime=5, UsePositionOffset=false, Write Material Random=true)]
  particleUpdateScript [&ParticleState()]
  rendererClasses [&NiagaraComponentRendererProperties]
  renderers [{rendererClass=&NiagaraComponentRendererProperties, rendererIndex=0}]
 }
 Glows {
  emitterUpdateScript [
   &EmitterState() {module=EmitterState001}
   &SpawnBurst_Instantaneous(Spawn Count=6, Use Spawn Probability=false, UseLoopCountLimit=true)
  ]
  particleSpawnScript [
   &FN_SetSpawnGroup(Number of Spawn Groups=3)
   &InitializeParticle(Color=<rgba 1 0.15 0 1>, Sprite Size=<30 120>, Sprite Size Mode=::NewEnumerator4, UsePositionOffset=false, Write Material Random=true)
   /SetVariables(
    Particles.EngineGlowCameraOffset:NiagaraFloat=&FN_FrontEndFloatModifier(Gameplay=12, Front End=12)
    Particles.EngineGlowColor:LinearColor=<rgba 1.75 0.4375 0 1>
   ) {isSetParameters=true}
   &CameraOffset(Camera Offset Amount=&FN_FrontEndFloatModifier(Gameplay=12, Front End=0))
  ]
  particleUpdateScript [
   &ParticleState(Kill Particles When Lifetime Has Elapsed=false, Let Infinitely Lived Particles Die When Emitter Deactivates=true)
   &SkeletalMeshLocation(
    Bone / Socket Index=&Add_Integer(A=&ModuloInt(A=&ReturnParticleIDAttributes(), B=6), B=4)
    MaskSpawnGroup=false
    Mesh Sampling Type=::NewEnumerator5
    Orientation Sampling=::NewEnumerator1
    Sampled Position Offset=/Niagara/DynamicInputs/Arrays/SelectPositionFromArray(
     Array Sampling Mode:ENiagaraArraySamplingMode=::NewEnumerator1
     Direct Array Index:NiagaraInt32=$Particles.SpawnGroup
     Interpolation:NiagaraFloat=?
     Direct Array Mode:ENiagara_WrapClamp=::NewEnumerator0
     Evaluation Type:ENiagaraRandomnessEvaluation=::NewEnumerator0
     Normalized Interpolation:NiagaraBool=true
     Recalculate Random Each Loop:NiagaraBool=false
     Position Array:NiagaraDataInterfaceArrayPosition=DataInterface() {propertyValues {GpuSyncMode=SyncCpuToGpu, MaxElements=0, PositionData [<0 5 0>, <0 5 0>, <0 5 0>]}}
     Randomness Mode:ENiagaraRandomnessMode=::NewEnumerator0
     Fixed Random Seed:NiagaraBool=false
     Random Seed:NiagaraInt32=?
    )
    Sampled Position Offset Coordinate Space:ENiagaraSkelMeshTransforms=::NewEnumerator2
    Skeletal Mesh=$System.Skeletal Mesh
    Socket Sampling Mode=::NewEnumerator5
   )
   &ScaleSpriteSize(
    Non-Uniform Curve Index:NiagaraFloat=$System.Timeline.Time
    Non-Uniform Curve Sprite Scale=DataInterface() {propertyValues {
      XCurve {keys [{leaveTangentWeight=0.3333, time=0, value=0.5}, {arriveTangentWeight=0.3333, time=0.15, value=1}]}
      YCurve {keys [
        {arriveTangent=17.5491, interpMode=RCIM_Cubic, leaveTangent=17.5491, leaveTangentWeight=0.4329, tangentMode=RCTM_User, time=0, value=0}
        <at 0.15 1 RCIM_Cubic>
       ]}
     }}
    Scale Sprite Size Mode=::NewEnumerator3
   ) {module=ScaleSpriteSize001}
   &ScaleColor(
    Scale Alpha=&Waveform(
     Global Amplitude Scale=&FNTimeOfDayScalar(Night Scale=0.6, Morning Scale=0.9, Evening Scale=0.9)
     "Waveform [1]"=::NewEnumerator9
     "[1] Frequency"=&RandomRangeFloat(Minimum=10, Maximum=20)
     "[1] Amplitude Min/Max"=<1 1.5>
     "[1] Scale"=&FN_FrontEndFloatModifier(Gameplay=&Lerp_Float(A=0.5, B=1, Alpha=$User.ForwardDot), Front End=1)
    )
    ScaleA=true
    ScaleRGB=false
   )
   &SpriteFacingAndAlignment(
    Sprite Alignment=&ReturnMeshOrientationAxis(Mesh Orientation Axis=::XAxis, Mesh Orientation Quat=$Particles.SkeletalMeshLocation.SampledRotation)
    Sprite Facing=<0 1 0>
    WriteAlignment=true
    WriteFacing=true
   )
   /SetVariables(
    Particles.EngineGlowColor:LinearColor=&MakeLinearColorFromVectorAndFloat(
     "Vector (RGB)"=&MakeVectorFromLinearColorRGB(Color=&MultiplyLinearColor(
       Linear Color=$Particles.Initial.EngineGlowColor
       Scale Factor=&MakeFloat4FromVectorAndFloat(
        "Vector (RGB)"=&VectorFromFloat(Value=&FloatFromCurve(
          FloatCurve=DataInterface() {propertyValues {Curve {keys [
              {arriveTangent=9.0725, interpMode=RCIM_Cubic, time=0, value=1}
              {arriveTangent=-12.2639, interpMode=RCIM_Cubic, leaveTangent=-12.2639, tangentMode=RCTM_User, time=0.2, value=5}
              {
               arriveTangentWeight=0.8477
               interpMode=RCIM_Cubic
               leaveTangentWeight=0.8477
               tangentMode=RCTM_User
               tangentWeightMode=RCTWM_WeightedBoth
               time=1
               value=1
              }
             ]}}}
          CurveIndex=$Emitter.Age
         ))
        "Float (Alpha)"=1
       )
      ))
     "Float (Alpha)"=&Multiply_Float(
      A=&Waveform(
       "Waveform [1]"=::NewEnumerator9
       "[1] Frequency"=&RandomRangeFloat(Minimum=10, Maximum=15)
       "[1] Amplitude Min/Max"=<0.75 1.25>
       "[1] Scale"=&ClampFloat(
        Float=&FN_FrontEndFloatModifier(Gameplay=&Lerp_Float(A=0.5, B=1, Alpha=$User.ForwardDot), Front End=1)
        Min=&FNTimeOfDayScalar(Night Scale=0.3, Morning Scale=0.45, Day Scale=0.5, Evening Scale=0.45)
        Max=&FNTimeOfDayScalar(Night Scale=0.6, Morning Scale=0.9, Evening Scale=0.9)
       )
      )
      B=&FN_FrontEndFloatModifier(Gameplay=1, Front End=0.35)
     )
    )
    Particles.EngineGlowSpriteSize:Vector2f=<40 40>
   ) {isSetParameters=true}
   &ScaleColor(
    Scale RGB=&VectorFromFloat(Value=&FloatFromCurve(
      FloatCurve=DataInterface() {propertyValues {Curve {keys [
          {arriveTangent=9.0725, interpMode=RCIM_Cubic, leaveTangent=9.0725, tangentMode=RCTM_User, time=0, value=1}
          {arriveTangent=-94.3756, interpMode=RCIM_Cubic, leaveTangent=-94.3756, tangentMode=RCTM_User, time=0.2, value=25}
          {
           arriveTangentWeight=0.702
           interpMode=RCIM_Cubic
           leaveTangentWeight=0.702
           tangentMode=RCTM_User
           tangentWeightMode=RCTWM_WeightedBoth
           time=1
           value=1
          }
         ]}}}
      CurveIndex=$Engine.System.Age
     ))
    ScaleA=true
    ScaleRGB=true
   ) {module=ScaleColor001}
   &FN_Light_Attributes(
    Color=$Particles.EngineGlowColor
    Create Color Attribute=true
    Create Exponent Attribute=true
    Create Radius Attribute=true
    Exponent=1
    Radius=50
    Scale Color=3.5
   )
   /SetVariables(Particles.SpriteSize2=&Multiply_Vector2D_ByFloat(Float=0.5, Vector2D=$Particles.EngineGlowSpriteSize)) {isSetParameters=true}
  ]
  rendererClasses [&NiagaraSpriteRendererProperties, &NiagaraLightRendererProperties]
  renderers [
   {rendererClass=&NiagaraSpriteRendererProperties}
   {rendererClass=&NiagaraSpriteRendererProperties, rendererIndex=1}
   {rendererClass=&NiagaraLightRendererProperties, rendererIndex=2}
   {rendererClass=&NiagaraSpriteRendererProperties, rendererIndex=3}
  ]
 }
 SideWingsinner {
  emitterSpawnScript [&FN_RibbonIDOffsetUtility()]
  emitterUpdateScript [
   &EmitterState(Life Cycle Mode=::NewEnumerator1, Loop Behavior=::NewEnumerator1, Loop Delay=1, Loop Duration Mode=::NewEnumerator1, UseLoopDelay=true)
   &FN_SpawnAnimTrail(Number of Anim Trails=2)
  ]
  particleSpawnScript [
   &FN_InitializeAnimTrailBoneIndecies()
   &FN_InitializeAnimTrailBoneIndecies(AnimTrail Selection=1, First Bone or Socket Index=2, Second Bone or Socket Index=3) {module=FN_InitializeAnimTrailBoneIndecies001}
   &FN_InitializeAnimTrail(Color=<rgba 0.85 0.8951 1 0.25>, Lifetime=0.25, SkeletalMesh=$System.Skeletal Mesh)
   &AddVelocity(Velocity=&Multiply_Vector(
     A=&MakeCustomVectorFromBool(Bool=$User.bIsFrontEnd, True Vector=<0 2500 0>, False Vector=<0 2500 0>)
     B=&VectorFromFloat(Value=0.3)
    )) {module=AddVelocity001}
  ]
  particleUpdateScript [
   &InheritSourceMovement(
    Applied Position Scale=&MakeVector(
     X=&FloatFromCurve(
      FloatCurve=DataInterface() {propertyValues {Curve {keys [
          {arriveTangent=-2.6336, interpMode=RCIM_Cubic, time=0, value=1}
          {arriveTangent=-2.5605, interpMode=RCIM_Cubic, leaveTangent=-2.5605, tangentMode=RCTM_User, time=1, value=0}
         ]}}}
      Scale Curve=-0.25
     )
     Y=&FloatFromCurve(
      FloatCurve=DataInterface() {propertyValues {Curve {keys [
          {arriveTangent=-2.7167, interpMode=RCIM_Cubic, time=0, value=1}
          {arriveTangent=-2.6803, interpMode=RCIM_Cubic, leaveTangent=-2.6803, tangentMode=RCTM_User, time=1, value=0}
         ]}}}
      Scale Curve=0
     )
     Z=-0.2
    )
    ApplyForce=false
    ApplyPosition=true
    ApplyPositionFalloff=false
    Source Velocity:Vector3f=&TransformVector(Vector=$Engine.Owner.Velocity, Source Space=::World, Destination Space=::Simulation)
   )
   &SolveForcesAndVelocity(Clamp Velocity=false, Limit Acceleration=false)
   &FN_UpdateAnimTrail(SkeletalMesh=$System.Skeletal Mesh) {module=FN_UpdateAnimTrail001}
   /SetVariables(Particles.RibbonU0Override=$Particles.NormalizedAge) {isSetParameters=true}
   &ScaleColor(
    Scale Alpha=&FN_FrontEndFloatModifier(Gameplay=&Lerp_Float(A=0, B=0.8, Alpha=$System.Timeline.NormalizedTime), Front End=1)
    ScaleA=true
    ScaleRGB=true
   )
  ]
  rendererClasses [&NiagaraRibbonRendererProperties]
  renderers [{rendererClass=&NiagaraRibbonRendererProperties}]
 }
 Smoke {
  emitterUpdateScript [
   &EmitterState()
   &SpawnRate(
    Spawn Probability=&FN_FrontEndFloatModifier(Gameplay=/Niagara/DynamicInputs/Integer/V2/Round_Float(Float:NiagaraFloat=$User.ForwardDot), Front End=1)
    SpawnRate=&Multiply_Float(A=6, B=35)
    Use Spawn Probability=true
   )
  ]
  particleSpawnScript [
   &PartitionParticles(Number Of Partitions=6)
   &InitializeParticle(
    Color=&MultiplyLinearColor(
     Linear Color=<rgba 1 1 1 1>
     Scale Factor=&MakeVector4_FromFloat(Float=&FN_FrontEndFloatModifier(Gameplay=1, Front End=1))
    )
    Lifetime Max=0.7
    Lifetime Min=0.5
    Lifetime Mode=::NewEnumerator1
    Material Random=1
    Mesh Scale=<0.4 0.35 0.35>
    Mesh Scale Mode=::NewEnumerator4
    UsePositionOffset=false
    Write Material Random=true
   )
   &AddVelocity(
    Scale Added Velocity=&Multiply_VectorByFloat(Vector=&Multiply_VectorByFloat(Vector=<2.5 2.5 2.5>, Float=1), Float=0.6)
    Velocity=&Multiply_VectorByFloat(
     Vector=&ReturnMeshOrientationAxis(Mesh Orientation Axis=::YAxis, Mesh Orientation Quat=$Particles.SkeletalMeshLocation.SampledRotation)
     Float=&Lerp_Float(A=100, B=400, Alpha=$System.EffectIntensity)
    )
   ) {enabled=false}
   &AddVelocityInCone(
    Cone Angle=50
    Cone Axis=&ReturnMeshOrientationAxis(Mesh Orientation Axis=::YAxis, Mesh Orientation Quat=$Particles.SkeletalMeshLocation.SampledRotation)
    Module.Use Velocity Falloff On Cone Axis=false
    Velocity Strength=66.6
   ) {enabled=false, module=AddVelocityInCone001}
   &InitialMeshOrientation(
    Mesh Orientation Mode=::NewEnumerator2
    Rotation=&RandomRangeVector(Minimum=<0 0 0>, Maximum=<1 0 0>)
    Use Rotation Vector=true
   ) {module=InitialMeshOrientation001}
   &CameraOffset(Camera Offset Amount=15)
  ]
  particleUpdateScript [
   &Drag(Use Linear Drag=true, Use Rotational Drag=true)
   &SkeletalMeshLocation(
    Bone / Socket Index=&SelectIntFromArray(Direct Array Index=$Particles.PartitionParticles.Partition, Int Selection Array=DataInterface() {propertyValues {IntData [4,5,6,7,8,9]}})
    MaskSpawnGroup=false
    Mesh Sampling Type=::NewEnumerator5
    Orientation Sampling=::NewEnumerator1
    Sampled Position Offset=&MakeVector(
     X=&FloatFromCurve(
      FloatCurve=DataInterface() {propertyValues {Curve {keys [{tangentMode=RCTM_User, time=0, value=0}, {tangentMode=RCTM_User, time=1, value=1}]}}}
      Scale Curve=100
     )
     Y=0
     Z=0
    )
    Skeletal Mesh=$System.Skeletal Mesh
    Socket Sampling Mode=::NewEnumerator5
   )
   &SolveForcesAndVelocity(Clamp Velocity=false, Limit Acceleration=false)
   &ParticleState()
   &UpdateMeshOrientation(
    Facing Direction=&ReturnMeshOrientationAxis(Mesh Orientation Axis=::XAxis, Mesh Orientation Quat=$Particles.SkeletalMeshLocation.SampledRotation)
    Facing Interpolation=false
    Orientation Method=::NewEnumerator1
    Override Quat=false
    TransformFacing=false
    TransformUp=false
   )
   &ScaleMeshSize(Scale Factor=&Multiply_VectorByFloat(
     Vector=&Multiply_VectorByFloat(
      Vector=&Multiply_Vector(
       A=<1 1 0.75>
       B=&VectorFromCurve(VectorCurve=DataInterface() {propertyValues {
          XCurve {keys [
            {arriveTangent=0.6082, interpMode=RCIM_Cubic, leaveTangent=0.6082, tangentMode=RCTM_User, time=0, value=1}
            {arriveTangent=0.25, interpMode=RCIM_Cubic, leaveTangent=0.25, time=0.2, value=1.1}
            <at 1 1.25 RCIM_Cubic>
           ]}
          YCurve {keys [<at 0 1>,<at 1 1>]}
          ZCurve {keys [<at 0 1>,<at 1 1>]}
         }})
      )
      Float=&RandomRangeFloat(Minimum=1, Maximum=1.5)
     )
     Float=&FloatFromCurve(FloatCurve=DataInterface() {propertyValues {Curve {keys [
          {arriveTangent=0.0836, interpMode=RCIM_Cubic, leaveTangent=0.0836, tangentMode=RCTM_User, time=0, value=0.5}
          {arriveTangent=8.8185, interpMode=RCIM_Cubic, leaveTangent=8.8185, tangentMode=RCTM_User, time=1, value=4}
         ]}}})
    ))
   &DynamicMaterialParameters(
    Index 0 Param 1=&FloatFromCurve(
     FloatCurve=DataInterface() {propertyValues {Curve {keys [
         {interpMode=RCIM_Cubic, tangentMode=RCTM_User, time=0, value=0}
         <at 0.2 0.55>
         {arriveTangent=-1.0223, interpMode=RCIM_Cubic, leaveTangent=-1.0223, tangentMode=RCTM_User, time=1, value=1}
        ]}}}
     Scale Curve=&RandomRangeFloat(Minimum=0.7, Maximum=1.5)
    )
    Index 0 Param 3=&UniformAOrBFloat(A=1, Distribution Weight=1, Evaluation Type=::NewEnumerator0, Recalculate Random Each Loop=true, Override Seed=false)
    Index 0 Param 4=&Lerp_Float(
     A=0.6
     B=0
     Alpha=&FloatFromCurve(FloatCurve=DataInterface() {propertyValues {Curve {keys [
          {arriveTangent=0.7438, interpMode=RCIM_Cubic, leaveTangent=0.7438, tangentMode=RCTM_User, time=0, value=0}
          {arriveTangent=0.0023, interpMode=RCIM_Cubic, leaveTangent=0.0023, tangentMode=RCTM_User, time=1, value=1}
         ]}}})
    )
    Index 1 Param 1=&FloatFromCurve(FloatCurve=DataInterface() {propertyValues {Curve {keys [<at 0 0.5 RCIM_Cubic>,<at 0.75 0 RCIM_Cubic>]}}}, Scale Curve=500)
    Module.Index0Param0Write=true
    Module.Index0Param1Write=true
    Module.Index0Param2Write=true
    Module.Index0Param3Write=true
    Module.Index1Param0Write=true
    Module.Index1Param1Write=true
    Module.Index1Param2Write=true
    Module.Index1Param3Write=true
    Write Parameter Index 1=true
   ) {module=DynamicMaterialParameters001}
   &ScaleColor(
    Scale Alpha=&FloatFromCurve(
     FloatCurve=DataInterface() {propertyValues {Curve {keys [
         {arriveTangent=2.7764, interpMode=RCIM_Cubic, leaveTangent=2.7764, tangentMode=RCTM_User, time=0, value=0}
         {interpMode=RCIM_Cubic, leaveTangent=-1.4999, tangentMode=RCTM_Break, time=0.2, value=1}
         {arriveTangent=-2.5126, tangentMode=RCTM_User, time=1, value=0}
        ]}}}
     Scale Curve=&Multiply_Float(A=0.1, B=$System.Timeline.NormalizedTime)
    )
    Scale RGB=&Multiply_Vector(
     A=&MakeVectorFromLinearColorRGB(Color=&ColorFromCurve(ColorCurve=DataInterface() {propertyValues {
         AlphaCurve {keys []}
         BlueCurve {keys [<at 0.0312 4.6875 RCIM_Cubic>,<at 0.2513 40 RCIM_Cubic>]}
         GreenCurve {keys [<at 0.0312 22.2656 RCIM_Cubic>,<at 0.2513 42.5 RCIM_Cubic>]}
         RedCurve {keys [<at 0.0312 75 RCIM_Cubic>,<at 0.2513 50 RCIM_Cubic>]}
        }}))
     B=&MakeVectorFromLinearColorRGB(Color=&FNTimeOfDayColor(Night Color=<rgba 1 0.8831 0.6452 1>, Morning Color=<rgba 1 1 1 1>, Day Color=<rgba 1 1 1 1>, Evening Color=<rgba 1 1 1 1>))
    )
    ScaleA=true
    ScaleRGB=true
   )
  ]
  rendererClasses [&NiagaraMeshRendererProperties]
  renderers [{rendererClass=&NiagaraMeshRendererProperties, rendererIndex=0}]
 }
 SmokeLarge_Local {
  bEnabled=false
  emitterUpdateScript [
   &EmitterState()
   &SpawnRate(
    Spawn Probability=&FN_FrontEndFloatModifier(Gameplay=&Round_Float(Float:NiagaraFloat=$User.ForwardDot), Front End=1)
    SpawnRate=&Multiply_Float(A=6, B=150)
    Use Spawn Probability=true
   )
   &SpawnRate(SpawnRate=&Multiply_Float(A=6, B=60), Use Spawn Probability=true) {enabled=false, module=SpawnRate001}
   &SpawnPerFrame(
    Spawn Count=6
    Spawn Probability:NiagaraFloat=&FN_FrontEndFloatModifier(Gameplay=$User.ForwardDot, Front End=1)
    Use Spawn Probability=true
   ) {enabled=false}
  ]
  particleSpawnScript [
   &FN_SetSpawnGroup(Number of Spawn Groups=3)
   &InitializeParticle(
    Color=<rgba 1 1 1 0>
    Lifetime Max=0.2
    Lifetime Min=0.2
    Lifetime Mode=::NewEnumerator1
    Mesh Scale=<0.01 0.01 0.01>
    Mesh Scale Mode=::NewEnumerator4
    Sprite Rotation Mode=::NewEnumerator2
    Sprite Size Mode=::NewEnumerator3
    Sprite UV Mode=::NewEnumerator3
    Uniform Sprite Size=&Multiply_Float(A=30, B=&MakeFloatFromVector(Vector=$Engine.Owner.Scale, Channel=::NewEnumerator0))
    UsePositionOffset=false
    Write Material Random=true
   )
   &CameraOffset(Camera Offset Amount=16)
   &AddVelocity(Velocity=&Multiply_Vector(
     A=&MakeCustomVectorFromBool(Bool=$User.bIsFrontEnd, True Vector=<0 -2500 0>, False Vector=<0 2500 0>)
     B=&VectorFromFloat(Value=0.25)
    )) {enabled=false, module=AddVelocity001}
   &AddVelocity(
    Rotation Coordinate Space=::World
    Velocity=&ReturnMeshOrientationAxis(Mesh Orientation Axis=::XAxis, Mesh Orientation Quat=$Particles.SkeletalMeshLocation.SampledRotation)
    Velocity Speed Scale=&FN_FrontEndFloatModifier(Gameplay=625, Front End=-625)
   ) {enabled=false}
  ]
  particleUpdateScript [
   &InheritSourceMovement(
    Applied Position Scale=&MakeVector(
     X=&FloatFromCurve(FloatCurve=DataInterface() {propertyValues {Curve {keys [<at 0 1>,<at 1 0.5>]}}})
     Y=&FloatFromCurve(FloatCurve=DataInterface() {propertyValues {Curve {keys [<at 0 1>,<at 1 0.5>]}}})
     Z=&FloatFromCurve(FloatCurve=DataInterface() {propertyValues {Curve {keys [<at 0 1>,<at 1 0.25>]}}})
    )
    ApplyForce=false
    ApplyPosition=true
    ApplyPositionFalloff=false
   ) {enabled=false}
   &ParticleState()
   &ScaleSpriteSize(
    Uniform Curve Scale=&RandomRangeFloat(Minimum=1, Maximum=1.5)
    Uniform Curve Sprite Scale=DataInterface() {propertyValues {Curve {keys [<at 0 0.4 RCIM_Cubic>,<at 1 1 RCIM_Cubic>]}}}
   )
   &SkeletalMeshLocation(
    Bone / Socket Index=&Add_Integer(A=&ModuloInt(A=&ReturnParticleIDAttributes(), B=6), B=4)
    MaskSpawnGroup=false
    Mesh Sampling Type=::NewEnumerator5
    Orientation Sampling=::NewEnumerator1
    Sampled Position Offset=&MakeVector(
     X=&FloatFromCurve(
      FloatCurve=DataInterface() {propertyValues {Curve {keys [{tangentMode=RCTM_User, time=0, value=0}, {tangentMode=RCTM_User, time=1, value=1}]}}}
      Scale Curve=80
     )
     Y=0
     Z=0
    )
    Skeletal Mesh=$System.Skeletal Mesh
    Socket Sampling Mode=::NewEnumerator5
   )
   &DynamicMaterialParameters(
    Index 0 Param 1=&Add_Float(
     A=&Multiply_Float(A=$Engine.Time, B=1)
     B=&RandomRangeFloat(Minimum=0, Maximum=1)
    )
    Index 0 Param 3=&FloatFromCurve(
     FloatCurve=DataInterface() {propertyValues {Curve {keys [{arriveTangent=-66.2764, interpMode=RCIM_Cubic, leaveTangent=-66.2764, tangentMode=RCTM_User, time=0, value=5}, <at 0.125 0>]}}}
     CurveIndex=$Particles.Age
     Scale Curve=&Multiply_Float(A=0.1, B=$System.Timeline.Time)
    )
    Index 0 Param 4=1
    Index 2 Param 1=1
    Index 2 Param 2=&RandomRangeFloat(Minimum=0, Maximum=1)
    Index 2 Param 3=&FloatFromCurve(FloatCurve=DataInterface() {propertyValues {Curve {keys [
         {interpMode=RCIM_Cubic, tangentMode=RCTM_Break, time=0.5, value=0}
         {arriveTangent=10.9325, interpMode=RCIM_Cubic, tangentMode=RCTM_Break, time=1, value=2}
        ]}}})
    Index 2 Param 4=0.5
    Module.Param0WriteEnabled=true
    Module.Param0WriteEnabled1=true
    Module.Param0WriteEnabled2=true
    Module.Param1WriteEnabled=true
    Module.Param1WriteEnabled1=true
    Module.Param1WriteEnabled2=true
    Module.Param2WriteEnabled=true
    Module.Param2WriteEnabled1=true
    Module.Param2WriteEnabled2=true
    Module.Param3WriteEnabled=true
    Module.Param3WriteEnabled1=true
    Module.Param3WriteEnabled2=true
    Write Parameter Index 1=true
    Write Parameter Index 2=true
   )
   &SkeletalMeshLocation(
    Bone / Socket Index=&Add_Integer(A=&ModuloInt(A=&ReturnParticleIDAttributes(), B=6), B=4)
    MaskSpawnGroup=false
    Mesh Sampling Type=::NewEnumerator5
    Orientation Sampling=::NewEnumerator1
    Sampled Position Offset Coordinate Space:ENiagaraSkelMeshTransforms=::NewEnumerator2
    Skeletal Mesh=$System.Skeletal Mesh
    Socket Sampling Mode=::NewEnumerator5
   ) {enabled=false, module=SkeletalMeshLocation001}
   /SetVariables(
    Particles.Color=&MultiplyLinearColor(
     Linear Color=&ColorFromCurve(ColorCurve=DataInterface() {propertyValues {
        AlphaCurve {keys [<at 0 0>,<at 0.1 0.4>,<at 0.5 0.4>,<at 1 0>]}
        BlueCurve {keys [<at 0.075 0.16>, {arriveTangent=0.0577, time=0.15, value=0.24}, <at 1 0.32>]}
        GreenCurve {keys [<at 0.075 0.2667>, {arriveTangent=-0.0961, time=0.15, value=0.28}, <at 1 0.34>]}
        RedCurve {keys [<at 0.075 0.8>, {arriveTangent=-1.0957, time=0.15, value=0.4}, <at 1 0.4>]}
       }})
     Scale Factor=&MakeFloat4FromVectorAndFloat(
      "Vector (RGB)"=&VectorFromFloat(Value=&FNTimeOfDayScalar(Night Scale=0.6, Morning Scale=0.9, Evening Scale=0.9))
      "Float (Alpha)"=$System.Timeline.Time
     )
    )
    Particles.Position=&AddVectorToPosition(
     Position=$Particles.Position
     Vector=&Multiply_VectorByFloat(
      Vector=&NormalizeVector(
       Vector To Normalize=&Lerp_Vector(
        A=<0 1 0>
        B=&TransformVector(
         Vector=&NormalizeVector(
          Vector To Normalize=&Multiply_Vector(A=&FN_FrontEndVector3Modifier(Gameplay=$Engine.Owner.Velocity, Front End=<0 0 0>), B=<-1 -1 -1>)
          Safe Normalize=true
          Safe Fallback Threshold=0
          Safe Fallback Vector=<0 0 0.001>
         )
         Source Space=::World
         Destination Space=::Local
        )
        Alpha=&FloatFromCurve(
         FloatCurve=DataInterface() {propertyValues {Curve {keys [{interpMode=RCIM_Cubic, leaveTangentWeight=81.8242, tangentMode=RCTM_Break, time=0, value=0}, {tangentMode=RCTM_User, time=1, value=1}]}}}
         Scale Curve=$System.ForwardAndRightDot
        )
       )
       Safe Normalize=true
       Safe Fallback Threshold=0
       Safe Fallback Vector=<0 0 0.001>
      )
      Float=&RandomRangeFloat(Minimum=8, Maximum=10)
     )
    )
   ) {enabled=false, isSetParameters=true}
   /SetVariables(Particles.Color=&MultiplyLinearColor(
     Linear Color=&ColorFromCurve(ColorCurve=DataInterface() {propertyValues {
        AlphaCurve {keys [
          <at 0 0 RCIM_Cubic>
          {arriveTangentWeight=0.0373, interpMode=RCIM_Cubic, leaveTangentWeight=0.1491, tangentMode=RCTM_User, time=0.1, value=0.2}
          {arriveTangentWeight=0.1384, interpMode=RCIM_Cubic, leaveTangentWeight=0.173, tangentMode=RCTM_User, time=0.5, value=0.2}
          <at 1 0 RCIM_Cubic>
         ]}
        BlueCurve {keys [<at 0.2 0.15>, {arriveTangent=0.0577, time=0.3, value=0.25}, <at 1 0.32>]}
        GreenCurve {keys [<at 0.2 0.25>, {arriveTangent=-0.0961, time=0.3, value=0.3}, <at 1 0.34>]}
        RedCurve {keys [<at 0.2 0.8>, {arriveTangent=-1.0957, time=0.3, value=0.4}, <at 1 0.4>]}
       }})
     Scale Factor=&MakeFloat4FromVectorAndFloat(
      "Vector (RGB)"=&VectorFromFloat(Value=&FNTimeOfDayScalar(Night Scale=0.6, Morning Scale=0.9, Evening Scale=0.9))
      "Float (Alpha)"=$System.Timeline.Time
     )
    )) {isSetParameters=true}
   &SolveForcesAndVelocity(Clamp Velocity=false, Limit Acceleration=false)
  ]
  rendererClasses [&NiagaraSpriteRendererProperties]
  renderers [
   {rendererClass=&NiagaraSpriteRendererProperties}
   {rendererClass=&NiagaraSpriteRendererProperties, rendererIndex=1}
  ]
 }
 ThrusterCore_Mesh {
  emitterUpdateScript [
   &EmitterState()
   &SpawnBurst_Instantaneous(Spawn Count=6, Use Spawn Probability=false, UseLoopCountLimit=true)
  ]
  particleSpawnScript [
   &FN_SetSpawnGroup(Number of Spawn Groups=3)
   &InitializeParticle(Color=<rgba 1 1 1 2>, Mesh Scale=<0.55 0.5 0.55>, Mesh Scale Mode=::NewEnumerator4, UsePositionOffset=false, Write Material Random=true)
   &InitialMeshOrientation(
    Mesh Orientation Mode=::NewEnumerator7
    Rotation=&RandomRangeVector(Minimum=<0 0 0>, Maximum=<0 1 0>)
    Use Rotation Vector=true
   )
   &InitialMeshOrientation(Mesh Orientation Mode=::NewEnumerator7, Rotation=<0 0 -0.25>, Use Rotation Vector=true) {module=InitialMeshOrientation001}
  ]
  particleUpdateScript [
   &ParticleState(Kill Particles When Lifetime Has Elapsed=false, Let Infinitely Lived Particles Die When Emitter Deactivates=true)
   &SkeletalMeshLocation(
    Bone / Socket Index=&Add_Integer(A=&ModuloInt(A=&ReturnParticleIDAttributes(), B=6), B=4)
    MaskSpawnGroup=false
    Mesh Sampling Type=::NewEnumerator5
    Sampled Position Offset=<0 -3 0>
    Sampled Position Offset Coordinate Space:ENiagaraSkelMeshTransforms=::NewEnumerator2
    Skeletal Mesh=$System.Skeletal Mesh
    Socket Sampling Mode=::NewEnumerator5
   )
   &ScaleMeshSize(Scale Factor=&VectorFromCurve(
     VectorCurve=DataInterface() {propertyValues {
       XCurve {keys [{interpMode=RCIM_Cubic, leaveTangentWeight=0.3257, tangentMode=RCTM_User, time=0, value=1}, <at 0.5 1>]}
       YCurve {keys [{arriveTangent=3.7338, interpMode=RCIM_Cubic, leaveTangent=3.7338, tangentMode=RCTM_User, time=0, value=1}, <at 0.5 5>]}
       ZCurve {keys [{interpMode=RCIM_Cubic, leaveTangentWeight=0.3257, tangentMode=RCTM_User, time=0, value=1}, <at 0.5 1>]}
      }}
     CurveIndex=&ClampFloat(Float=$System.Timeline.Time, Min=0.001)
     Scale Curve=&Multiply_VectorByFloat(
      Vector=&Multiply_VectorByFloat(
       Vector=&MakeVector(
        X=1
        Y=&Waveform(
         "Waveform [1]"=::NewEnumerator9
         "[1] Frequency"=&RandomRangeFloat(Minimum=25, Maximum=35)
         "[1] Amplitude Min/Max"=<0.6 1.2>
        )
        Z=1
       )
       Float=1
      )
      Float=/Game/Effects/Niagara/DynamicInputs/FN_PickFloatFromSpawnGroupCutoff(Over Cutoff Float:NiagaraFloat=0.9, Under Cutoff Float:NiagaraFloat=1, SpawnGroup:NiagaraInt32=$Particles.SpawnGroup, Cutoff:NiagaraInt32=1)
     )
    ))
   &ScaleColor(
    Scale Alpha=&FN_FrontEndFloatModifier(Gameplay=1, Front End=0.5)
    Scale RGB=&VectorFromFloat(Value=&FNTimeOfDayScalar(Night Scale=0.6, Morning Scale=0.9, Evening Scale=0.9))
    ScaleA=true
    ScaleRGB=true
   ) {module=ScaleColor001}
   &ScaleMeshSize(Scale Factor=&MakeVector(
     X=1
     Y=&FloatFromCurve(
      FloatCurve=DataInterface() {propertyValues {Curve {keys [
          {arriveTangent=9.0725, interpMode=RCIM_Cubic, leaveTangent=9.0725, tangentMode=RCTM_User, time=0, value=1}
          {arriveTangent=-7.2652, interpMode=RCIM_Cubic, leaveTangent=-7.2652, tangentMode=RCTM_User, time=0.1408, value=5}
          {arriveTangent=-0.0405, interpMode=RCIM_Cubic, leaveTangent=-0.0405, tangentMode=RCTM_User, time=1, value=1}
         ]}}}
      CurveIndex=$Engine.System.Age
     )
     Z=1
    )) {module=ScaleMeshSize002}
   &ScaleColor(
    Scale RGB=&VectorFromFloat(Value=&FloatFromCurve(
      FloatCurve=DataInterface() {propertyValues {Curve {keys [
          {arriveTangent=9.0725, interpMode=RCIM_Cubic, leaveTangent=9.0725, tangentMode=RCTM_User, time=0, value=1}
          <at 0.2 2 RCIM_Cubic>
          {
           arriveTangentWeight=0.8201
           interpMode=RCIM_Cubic
           leaveTangentWeight=0.8201
           tangentMode=RCTM_User
           tangentWeightMode=RCTWM_WeightedBoth
           time=1
           value=1
          }
         ]}}}
      CurveIndex=$Engine.System.Age
     ))
    ScaleA=true
    ScaleRGB=true
   )
   /SetVariables(Particles.Scale=/Niagara/DynamicInputs/Math/Clamp_Vector(Vector:Vector3f=$Particles.Scale, Min:Vector3f=<0 0 0>, Max:Vector3f=<3 3 3>)) {isSetParameters=true}
  ]
  rendererClasses [&NiagaraMeshRendererProperties]
  renderers [{rendererClass=&NiagaraMeshRendererProperties, rendererIndex=0}]
 }
 Thrusters {
  emitterUpdateScript [
   &EmitterState()
   &SpawnRate(SpawnRate=&Multiply_Float(A=8, B=4), Use Spawn Probability=false)
  ]
  particleSpawnScript [
   &PartitionParticles(Number Of Partitions=4)
   &InitializeParticle(
    Color=<rgba 3 0.75 0 100>
    Mesh Scale=<1 0.1 0.1>
    Mesh Scale Mode=::NewEnumerator4
    Sprite Size Mode=::NewEnumerator3
    Uniform Sprite Size=25
    UsePositionOffset=false
    Write Material Random=true
   )
   /SetVariables(Particles.Dynam_Param2:Vector4f=<1 5 5 5>) {isSetParameters=true}
   &InitialMeshOrientation(Mesh Orientation Mode=::NewEnumerator7, Rotation=<0 0 0.5>, Use Rotation Vector=true)
   &DynamicMaterialParameters(
    Index 0 Param 1=2
    Index 0 Param 2=5
    Index 0 Param 3=5
    Index 0 Param 4=5
    Module.Index0Param0Write=true
    Module.Index0Param1Write=true
    Module.Index0Param2Write=true
    Module.Index0Param3Write=true
   )
  ]
  particleUpdateScript [
   &ParticleState()
   &InterpolateOverTime(Target Float Value:NiagaraFloat=&ClampFloat(Float=$User.ForwardDot))
   &SkeletalMeshLocation(
    Bone / Socket Index=&Add_Integer(A=&ModuloInt(A=&ReturnParticleIDAttributes(), B=6), B=4)
    MaskSpawnGroup=false
    Mesh Sampling Type=::NewEnumerator5
    Skeletal Mesh=$System.Skeletal Mesh
    Socket Sampling Mode=::NewEnumerator5
   )
   /SetVariables(Particles.InnerFlame_Color:LinearColor=&MultiplyLinearColor(
     Linear Color=&LerpLinearColors(
      StartColor=<rgba 1 0.25 0 50>
      EndColor=<rgba 3 0.75 0 50>
      LerpFactor=$Particles.InterpolateOverTime.InterpolateOverTime_Float.Moving Average
     )
     Scale Factor=&self:mulV4(
      Vector 4:Vector4f=&MakeVector4_FromFloat(Float=&FNTimeOfDayScalar(Night Scale=0.5, Morning Scale=1, Evening Scale=0.5))
      Vector 4001:Vector4f=&MakeFloat4FromVectorAndFloat("Vector (RGB)"=<1 1 1>, "Float (Alpha)"=1)
     )
    )) {isSetParameters=true}
   &ScaleColor(
    Scale Alpha=&Lerp_Float(
     A=&FNTimeOfDayScalar(Night Scale=2, Morning Scale=3, Day Scale=3, Evening Scale=2)
     B=0
     Alpha=$Particles.NormalizedAge
    )
    Scale RGB=&VectorFromFloat(Value=&Lerp_Float(A=0.05, B=0.1, Alpha=$Particles.InterpolateOverTime.InterpolateOverTime_Float.Moving Average))
    ScaleA=true
    ScaleRGB=true
   )
   &ScaleMeshSize(Scale Factor=&Lerp_Vector(A=<0.05 0.85 0.85>, B=<0.45 0.7 0.7>, Alpha=$Particles.InterpolateOverTime.InterpolateOverTime_Float.Moving Average)) {module=ScaleMeshSize001}
   &ScaleMeshSize(Scale Factor=&MakeVector(
     X=&FloatFromCurve(FloatCurve=DataInterface() {propertyValues {Curve {keys [<at 0 0.5 RCIM_Cubic>, {arriveTangent=0.25, interpMode=RCIM_Cubic, leaveTangent=0.25, time=0.5, value=1.5}, <at 1 0.75 RCIM_Cubic>]}}})
     Y=&FloatFromCurve(FloatCurve=DataInterface() {propertyValues {Curve {keys [<at 0.5 1 RCIM_Cubic>,<at 1 1.75 RCIM_Cubic>]}}}, Scale Curve=0.5)
     Z=&FloatFromCurve(FloatCurve=DataInterface() {propertyValues {Curve {keys [<at 0.5 1 RCIM_Cubic>,<at 1 1.75 RCIM_Cubic>]}}}, Scale Curve=0.5)
    ))
   &ScaleColor(
    Scale RGB=&VectorFromFloat(Value=&FloatFromCurve(
      FloatCurve=DataInterface() {propertyValues {Curve {keys [
          {arriveTangent=9.0725, interpMode=RCIM_Cubic, leaveTangent=9.0725, tangentMode=RCTM_User, time=0, value=1}
          {arriveTangent=-109.8729, interpMode=RCIM_Cubic, leaveTangent=-109.8729, tangentMode=RCTM_User, time=0.2, value=5}
          {
           arriveTangentWeight=0.8201
           interpMode=RCIM_Cubic
           leaveTangentWeight=0.8201
           tangentMode=RCTM_User
           tangentWeightMode=RCTWM_WeightedBoth
           time=1
           value=1
          }
         ]}}}
      CurveIndex=$Engine.System.Age
     ))
    ScaleA=true
    ScaleRGB=true
   ) {module=ScaleColor001}
   &ScaleMeshSize(Scale Factor=&MakeVector(
     X=&FloatFromCurve(
      FloatCurve=DataInterface() {propertyValues {Curve {keys [
          {arriveTangent=9.0725, interpMode=RCIM_Cubic, leaveTangent=9.0725, tangentMode=RCTM_User, time=0, value=1}
          {arriveTangent=-14.3082, interpMode=RCIM_Cubic, leaveTangent=-14.3082, tangentMode=RCTM_User, time=0.1408, value=10}
          {arriveTangent=-0.0405, interpMode=RCIM_Cubic, leaveTangent=-0.0405, tangentMode=RCTM_User, time=1, value=1}
         ]}}}
      CurveIndex=$Engine.System.Age
     )
     Y=1
     Z=1
    )) {module=ScaleMeshSize002}
   /SetVariables(Particles.Scale=&Clamp_Vector(Vector:Vector3f=$Particles.Scale, Min:Vector3f=<0 0 0>, Max:Vector3f=<0.35 0.35 0.35>)) {isSetParameters=true}
  ]
  rendererClasses [&NiagaraMeshRendererProperties]
  renderers [
   {rendererClass=&NiagaraMeshRendererProperties, rendererIndex=0}
   {rendererClass=&NiagaraMeshRendererProperties, rendererIndex=1}
  ]
 }
}
summary {
 systemName=NS_Glider_RadiumFox_Trail
 emitters [
  {bEnabled=false, emitterName=Empty, rendererClasses [&NiagaraComponentRendererProperties]}
  {emitterName=SideWingsinner, rendererClasses [&NiagaraRibbonRendererProperties]}
  {emitterName=EmbersGPUHigh, simTarget=GPUComputeSim, rendererClasses [&NiagaraSpriteRendererProperties]}
  {bEnabled=false, emitterName=SmokeLarge_Local, rendererClasses [&NiagaraSpriteRendererProperties]}
  {emitterName=ThrusterCore_Mesh, rendererClasses [&NiagaraMeshRendererProperties]}
  {emitterName=Thrusters, rendererClasses [&NiagaraMeshRendererProperties]}
  {emitterName=Glows, rendererClasses [&NiagaraSpriteRendererProperties, &NiagaraLightRendererProperties]}
  {emitterName=Smoke, rendererClasses [&NiagaraMeshRendererProperties]}
 ]
 userVariables [
  User.bIsFrontEnd {type=NiagaraBool}
  User.bIsFrontEndPreview {type=NiagaraBool}
  User.bIsFullyDeployed {type=NiagaraBool, defaultValue=true}
  User.ColorVariantIndex {type=NiagaraInt32}
  User.ForwardDot {type=NiagaraFloat, defaultValue=1, description=""}
  User.RibbonIDOffset {type=NiagaraInt32}
  User.RightDot {type=NiagaraFloat, defaultValue=1, description=""}
  User.SkeletalMesh {
   type=NiagaraDataInterfaceSkeletalMesh
   defaultValue {
    struct=NiagaraExt_VariableValue_DataInterface
    value {dataInterface=&self:NiagaraDataInterfaceSkeletalMesh_0}
   }
  }
 ]
}
